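# Conversation database for The Origin NPC, defining its greeting and the dialogue branches that follow it.
# The entry point leads to a single Speech node that defaults to "Looper." and offers contextual lines for a first-ever conversation or for players wearing specific character outfits.
# Priority, and in some cases weight, is set on those lines to govern selection.
# The speech node then links out to the shared NPC quests and basic services conversations.
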ConversationDatabase /NPCLibrary/NPCs/TheOrigin/TheOrigin/NPCLibrary_Conversation_TheOrigin
entry_points [{tag=Conversation.Entry.TheOrigin, destinations [#1]}]
nodes [
 ConversationEntryPointNode_7 {
  id=#1
  class=ConversationEntryPointNode
  outputs [#2]
  properties {entryTag {tagName=Conversation.Entry.TheOrigin}}
 }
 Speech_C_4 {
  id=#2
  class=Speech_C
  outputs [#3, #4]
  properties {
   nodeName=""
   generalConfig {
    defaultMessage=Looper.
    contextualMessages [
     {
      message=Impressive work in the mech.
      contextRequirements [{requirement {class=FortControllerRequirement_IsFirstEverConversationWithNPC}}]
      priority {curve {rowName=Tandem.Speech.Priority.FirstEverInteraction}}
     }
     {
      message="Remember, Syndicator. Our Island has laws. "
      contextRequirements [{requirement {
         class=FortControllerRequirement_HasCID
         properties {allowedCIds [&CID_A_425_Athena_Commando_F_BlueJay]}
        }}]
      priority {value=0, curve {curveTable=null, rowName=None}}
     }
     {
      message=You were right about her. I'm sorry.
      contextRequirements [{requirement {
         class=FortControllerRequirement_HasCID
         properties {allowedCIds [/BRCosmetics/Athena/Items/Cosmetics/Characters/CID_540_Athena_Commando_M_MeteorManRemix]}
        }}]
      priority {value=0, curve {curveTable=null, rowName=None}}
     }
     {
      message="I was stupid. Please, never let me be that stupid again."
      contextRequirements [{requirement {
         class=FortControllerRequirement_HasCID
         properties {allowedCIds [/BRCosmetics/Athena/Items/Cosmetics/Characters/CID_547_Athena_Commando_F_Meteorwoman]}
        }}]
      priority {value=10, curve {curveTable=null, rowName=None}}
      weight {value=10}
     }
     {
      message=I've got my eye on you.
      contextRequirements [{requirement {
         class=FortControllerRequirement_HasCID
         properties {allowedCIds [/BRCosmetics/Athena/Items/Cosmetics/Characters/CID_A_374_Athena_Commando_F_Binary]}
        }}]
      priority {value=0, curve {curveTable=null, rowName=None}}
     }
     {
      message=Any activity on the perimeter? Stay vigilant.
      contextRequirements [{requirement {
         class=FortControllerRequirement_HasCID
         properties {allowedCIds [/BRCosmetics/Athena/Items/Cosmetics/Characters/CID_A_395_Athena_Commando_F_BinaryTwin]}
        }}]
      priority {value=0, curve {curveTable=null, rowName=None}}
     }
     {
      message=I can't help but think this isn't over.
      contextRequirements [{requirement {
         class=FortControllerRequirement_HasCID
         properties {allowedCIds [/BRCosmetics/Athena/Items/Cosmetics/Characters/CID_A_286_Athena_Commando_M_Turtleneck]}
        }}]
      priority {value=10, curve {curveTable=null, rowName=None}}
      weight {value=10}
     }
     {
      message="It's nice to see you settling down, my friend."
      contextRequirements [{requirement {
         class=FortControllerRequirement_HasCID
         properties {allowedCIds [/BRCosmetics/Athena/Items/Cosmetics/Characters/CID_140_Athena_Commando_M_Visitor]}
        }}]
      priority {value=0, curve {curveTable=null, rowName=None}}
     }
     {
      message=" I'm working on being more trusting. So... I trust you."
      contextRequirements [{requirement {
         class=FortControllerRequirement_HasCID
         properties {allowedCIds [&CID_883_Athena_Commando_M_ChOneJonesy]}
        }}]
      priority {value=0, curve {curveTable=null, rowName=None}}
     }
    ]
   }
  }
 }
 ConversationLinkNode_0 {id=#4, class=ConversationLinkNode, properties {remoteEntryTag {tagName=Conversation.Entry.BasicServices}}}
 ConversationLinkNode_1 {id=#3, class=ConversationLinkNode, properties {remoteEntryTag {tagName=Conversation.Entry.NPCLibrary.Quests}}}
]
